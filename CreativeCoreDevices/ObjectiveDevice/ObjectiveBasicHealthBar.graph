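# Health bar and indicator widget for a Device_Objective_V2, showing icon, label and health with team-affiliation colors.
# Objective and view-target delegates update health, colors and damage animations.
# A staggered looping timer drives visibility from minigame state, mesh visibility and hide rules, animating show/hide and rotating an off-screen arrow when clamped.
# An optional custom widget loads asynchronously and receives the same data through an MVVM view model.
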WidgetBlueprint /CreativeCoreDevices/ObjectiveDevice/ObjectiveBasicHealthBar
animations [
 Appear {
  binding_count=4
  length_seconds=0.7
  object=&self:Appear
  sequence {bindings [
    ProgressIcon {
     id=#1
     possessed_class=Image
     tracks [
      {
       type=MovieSceneFloatTrack
       display_name=Render Opacity
       sections [{
         type=MovieSceneFloatSection
         end_frame=10
         start_frame=0
         channels [None {type=MovieSceneScriptingFloatChannel, keys [[0,0],[9.789,1],[10.789,1]]}]
        }]
      }
      {
       type=MovieScene2DTransformTrack
       sections [{
         type=MovieScene2DTransformSection
         end_frame=10
         start_frame=0
         channels [
          Translation.X {type=MovieSceneScriptingFloatChannel}
          Translation.Y {type=MovieSceneScriptingFloatChannel}
          Angle {type=MovieSceneScriptingFloatChannel}
          Scale.X {type=MovieSceneScriptingFloatChannel, keys [[2.6977,-0.5],[7.789,1.05],[10.789,1]]}
          Scale.Y {type=MovieSceneScriptingFloatChannel, keys [[2.6977,0.5],[7.789,1.05],[10.789,1]]}
          Shear.X {type=MovieSceneScriptingFloatChannel}
          Shear.Y {type=MovieSceneScriptingFloatChannel}
         ]
        }]
      }
      {
       type=MovieSceneWidgetMaterialTrack
       sections [{
         type=MovieSceneParameterSection
         end_frame=10
         start_frame=0
         channels [OverrideFillColor {type=MovieSceneScriptingFloatChannel, keys [[0,1],[4.8253,0.9763],[5.327,-0.0237],[9.199,0],[10.789,0]]}]
        }]
      }
     ]
    }
    HealthBar {
     id=#2
     possessed_class=HorizontalBox
     tracks [
      {
       type=MovieSceneFloatTrack
       display_name=Render Opacity
       sections [{
         type=MovieSceneFloatSection
         end_frame=14
         start_frame=0
         channels [None {type=MovieSceneScriptingFloatChannel, keys [[0,0],[10.789,0],[14,1]]}]
        }]
      }
      {
       type=MovieScene2DTransformTrack
       sections [{
         type=MovieScene2DTransformSection
         end_frame=14
         start_frame=10
         channels [
          Translation.X {type=MovieSceneScriptingFloatChannel, const [0,2]}
          Translation.Y {type=MovieSceneScriptingFloatChannel, keys [[10,7],[14,0]]}
          Angle {type=MovieSceneScriptingFloatChannel}
          Scale.X {type=MovieSceneScriptingFloatChannel}
          Scale.Y {type=MovieSceneScriptingFloatChannel}
          Shear.X {type=MovieSceneScriptingFloatChannel}
          Shear.Y {type=MovieSceneScriptingFloatChannel}
         ]
        }]
      }
     ]
    }
    Custom {
     id=#3
     display_name=Overlay_0
     possessed_class=Overlay
     tracks [
      {
       type=MovieScene2DTransformTrack
       sections [{
         type=MovieScene2DTransformSection
         end_frame=10
         start_frame=0
         channels [
          Scale.X {type=MovieSceneScriptingFloatChannel, keys [[0,1],[3,0.8665],[4,1.4397],[6.139,1],[10.789,1]]}
          Scale.Y {type=MovieSceneScriptingFloatChannel, keys [[0,1],[3,1.5185],[4,0.8081],[6.139,1],[10.789,1]]}
         ]
        }]
      }
      {
       type=MovieSceneFloatTrack
       display_name=Render Opacity
       sections [{
         type=MovieSceneFloatSection
         end_frame=10
         start_frame=0
         channels [None {type=MovieSceneScriptingFloatChannel, keys [[0,0],[5,1],[10.789,1]]}]
        }]
      }
     ]
    }
    IndicatorIcon {
     id=#4
     possessed_class=Image
     tracks [
      {
       type=MovieScene2DTransformTrack
       sections [{
         type=MovieScene2DTransformSection
         end_frame=10
         start_frame=0
         channels [
          Scale.X {type=MovieSceneScriptingFloatChannel, keys [[0,1],[3,0.8665],[4,1.4397],[6.139,1],[10.789,1]]}
          Scale.Y {type=MovieSceneScriptingFloatChannel, keys [[0,1],[3,1.5185],[4,0.8081],[6.139,1],[10.789,1]]}
         ]
        }]
      }
      {
       type=MovieSceneFloatTrack
       display_name=Render Opacity
       sections [{type=MovieSceneFloatSection, end_frame=10, start_frame=0, channels [None {type=MovieSceneScriptingFloatChannel, const [1,3]}]}]
      }
     ]
    }
   ]}
 }
 FriendlyObjectiveDamaged {
  binding_count=3
  length_seconds=0.2
  object=&self:FriendlyObjectiveDamaged
  sequence {
   bindings [
    CurHP {
     id=#5
     display_name=HealthBar
     possessed_class=TextBlock
     tracks [{
       type=MovieScene2DTransformTrack
       sections [{
         type=MovieScene2DTransformSection
         end_frame=12
         start_frame=0
         channels [
          Translation.X {type=MovieSceneScriptingFloatChannel}
          Translation.Y {type=MovieSceneScriptingFloatChannel, keys [[1,0],[5,-4],[9,0]]}
          Angle {type=MovieSceneScriptingFloatChannel}
          Scale.X {type=MovieSceneScriptingFloatChannel}
          Scale.Y {type=MovieSceneScriptingFloatChannel}
          Shear.X {type=MovieSceneScriptingFloatChannel}
          Shear.Y {type=MovieSceneScriptingFloatChannel}
         ]
        }]
      }]
    }
    IndicatorIcon {
     id=#6
     possessed_class=Image
     tracks [{
       type=MovieScene2DTransformTrack
       sections [{
         type=MovieScene2DTransformSection
         end_frame=12
         start_frame=0
         channels [
          Scale.X {type=MovieSceneScriptingFloatChannel, keys [[0,1],[3,0.95],[6,1.125],[12,1]]}
          Scale.Y {type=MovieSceneScriptingFloatChannel, keys [[0,1],[3,0.95],[6,1.125],[12,1]]}
         ]
        }]
      }]
    }
    ProgressIcon {
     id=#7
     possessed_class=Image
     tracks [
      {
       type=MovieSceneWidgetMaterialTrack
       sections [{
         type=MovieSceneParameterSection
         end_frame=12
         start_frame=0
         channels [ShowDamage {type=MovieSceneScriptingFloatChannel, keys [[0,1],[12,0]]}]
        }]
      }
      {
       type=MovieScene2DTransformTrack
       sections [{
         type=MovieScene2DTransformSection
         end_frame=12
         start_frame=0
         channels [
          Translation.X {type=MovieSceneScriptingFloatChannel, const [0,1]}
          Translation.Y {type=MovieSceneScriptingFloatChannel, const [0,1]}
          Angle {type=MovieSceneScriptingFloatChannel, const [0,1]}
          Scale.X {type=MovieSceneScriptingFloatChannel, keys [[0,1],[3,1.125],[6,0.9],[12,1]]}
          Scale.Y {type=MovieSceneScriptingFloatChannel, keys [[0,1],[3,1.125],[6,0.9],[12,1]]}
          Shear.X {type=MovieSceneScriptingFloatChannel}
          Shear.Y {type=MovieSceneScriptingFloatChannel}
         ]
        }]
      }
     ]
    }
   ]
   display_rate {numerator=60}
  }
 }
 EnemyObjectiveDamaged {
  binding_count=3
  length_seconds=0.2
  object=&self:EnemyObjectiveDamaged
  sequence {
   bindings [
    IndicatorIcon {
     id=#8
     possessed_class=Image
     tracks [{
       type=MovieScene2DTransformTrack
       sections [{
         type=MovieScene2DTransformSection
         end_frame=12
         start_frame=0
         channels [
          Scale.X {type=MovieSceneScriptingFloatChannel, keys [[0,1],[3,0.95],[6,1.125],[12,1]]}
          Scale.Y {type=MovieSceneScriptingFloatChannel, keys [[0,1],[3,0.95],[6,1.125],[12,1]]}
         ]
        }]
      }]
    }
    ProgressIcon {
     id=#7
     possessed_class=Image
     tracks [
      {
       type=MovieSceneWidgetMaterialTrack
       sections [{
         type=MovieSceneParameterSection
         end_frame=12
         start_frame=0
         channels [ShowDamage {type=MovieSceneScriptingFloatChannel, keys [[0,1],[12,0]]}]
        }]
      }
      {
       type=MovieScene2DTransformTrack
       sections [{
         type=MovieScene2DTransformSection
         end_frame=12
         start_frame=0
         channels [
          Translation.X {type=MovieSceneScriptingFloatChannel}
          Translation.Y {type=MovieSceneScriptingFloatChannel}
          Angle {type=MovieSceneScriptingFloatChannel}
          Scale.X {type=MovieSceneScriptingFloatChannel, keys [[0,1],[3,1.1],[6,0.9],[12,1]]}
          Scale.Y {type=MovieSceneScriptingFloatChannel, keys [[0,1],[3,1.1],[6,0.9],[12,1]]}
          Shear.X {type=MovieSceneScriptingFloatChannel}
          Shear.Y {type=MovieSceneScriptingFloatChannel}
         ]
        }]
      }
     ]
    }
    CurHP {
     id=#9
     display_name=HealthBar
     possessed_class=TextBlock
     tracks [{
       type=MovieScene2DTransformTrack
       sections [{
         type=MovieScene2DTransformSection
         end_frame=12
         start_frame=0
         channels [
          Translation.X {type=MovieSceneScriptingFloatChannel}
          Translation.Y {type=MovieSceneScriptingFloatChannel, keys [[1,0],[5,-4],[9,0]]}
          Angle {type=MovieSceneScriptingFloatChannel}
          Scale.X {type=MovieSceneScriptingFloatChannel}
          Scale.Y {type=MovieSceneScriptingFloatChannel}
          Shear.X {type=MovieSceneScriptingFloatChannel}
          Shear.Y {type=MovieSceneScriptingFloatChannel}
         ]
        }]
      }]
    }
   ]
   display_rate {numerator=60}
  }
 }
]
graphs [
 CanDetermineVisibility {nodes [
   CanDetermineVisibility() {is=title K2Node_FunctionEntry}
   > n1 = Sequence()
   > n2 = Sequence()
   > n3 = Is Valid(InputObject=Get CurrentObjective() -> (CurrentObjective:Device Objective V2 Object Reference))
   Return Node!(execute=@n3.Is Not Valid, bIsDeterminable:bool=false)
   n4 = Is Valid(exec=@n2.then_1, InputObject=Get CurrentViewTarget() -> (CurrentViewTarget:Actor Object Reference))
   Return Node(execute=@n4.Is Not Valid, bIsDeterminable=false)
   Return Node(execute=@n1.then_1, bIsDeterminable=true)
  ]}
 EventGraph {nodes [
   OnLocalPlayerViewTargetChanged() {is=title K2Node_CustomEvent}
   > UnbindViewTargetEvents!(execute, self, ViewTarget:Actor Object Reference=Get CurrentViewTarget()) -> (then)
   > n1 = Set CurrentViewTarget(CurrentViewTarget:Actor Object Reference=GetViewTarget!() -> (ViewTarget:Actor Object Reference)) -> (Output_Get:Actor Object Reference)
   > BindViewTargetEvents!(execute, self, ViewTarget:Actor Object Reference=@n1) -> (then)
   UpdateVisibilityCheck() {is=title K2Node_CustomEvent}
   > n2 = Branch(Condition=ShouldHideWithoutAnimation!() -> (bReturnValue:bool))
   > n3 = Set PendingVisibility(PendingVisibility:ESlateVisibility Enum=Get HideVisibility() -> (HideVisibility:ESlateVisibility Enum=Visible)) -> (Output_Get:ESlateVisibility Enum=Visible)
   > SetVisibility(InVisibility=@n3)
   Update Offscreen Arrow(execute=@n2.else)
   > n4 = Set DesiredVisibility(DesiredVisibility:ESlateVisibility Enum=GetDesiredVisibility!() -> (ReturnValue:ESlateVisibility Enum=Visible)) -> (Output_Get:ESlateVisibility Enum=Visible)
   > Branch(Condition="Not Equal (Byte)"!(A:Byte=@n4, B:Byte=Get PendingVisibility() -> (PendingVisibility:ESlateVisibility Enum=Visible)) -> (ReturnValue:bool))
   > n5 = Set PendingVisibility(PendingVisibility=Get DesiredVisibility() -> (DesiredVisibility:ESlateVisibility Enum=Visible)) -> (Output_Get=Visible)
   n6 = "Equal (Byte)"!(A:Byte=@n5, B:Byte=Get ShowVisibility() -> (ShowVisibility:ESlateVisibility Enum=Visible)) -> (ReturnValue:bool)
   > PlayAnimation(
    InAnimation=Get Appear() -> (Appear:Widget Animation Object Reference)
    PlayMode=Select(Option 0:EUMGSequencePlayMode Enum=Reverse, Option 1:EUMGSequencePlayMode Enum, Index=@n6) -> (ReturnValue:EUMGSequencePlayMode Enum)
   )
   > n7 = Branch(Condition=@n6)
   > SetVisibility(InVisibility=Get PendingVisibility() -> (PendingVisibility=Visible))
   > Clear and Invalidate Timer by Handle(Handle=Get HideTimer() -> (HideTimer:Timer Handle Structure))
   n8 = HideIndicator() {is=title K2Node_CustomEvent}
   n9 = Set Timer by Event(execute=@n7.else, Delegate=@n8, Time=GetEndTime(self=Get Appear()), bLooping=false)
   > Set HideTimer(HideTimer:Timer Handle Structure=@n9) -> (Output_Get:Timer Handle Structure)
   Assign Indicator Icon() {is=title K2Node_CustomEvent}
   > n10 = Sequence!(execute) -> (then_0, then_1, then_2:Exec)
   > n11 = Get!(execute:Exec) -> (then:Exec, else:Exec, CurrentObjective:Device Objective V2 Object Reference)
   > n12 = Async Load Asset(Asset=Get Icon Library Texture with Empty Texture!(
     Icon Option:"Creative Icon Option Structure (by ref)"=Get Objective Identifier!(self:Device Objective V2 Object Reference=@n11) -> (Objective Identifier:Creative Icon Option Structure)
     Should Use Large Icon:bool=false
    ) -> (Icon Texture:Texture 2D Soft Object Reference) {is=title K2Node_CallFunction})
   n13 = Cast To Texture2D(execute=@n12.Completed, Object=@n12)
   > Set LoadedIconTexture(LoadedIconTexture:Texture 2D Object Reference=@n13) -> (Output_Get:Texture 2D Object Reference)
   > SetBrushFromTexture!(
    execute
    self:Image Object Reference=Get IndicatorIcon() -> (IndicatorIcon:Image Object Reference)
    Texture:Texture 2D Object Reference=@n13
    bMatchSize:bool=false
   ) -> (then)
   n14 = Get(execute=@n10.then_1)
   n15 = Get Custom Objective Text!(self:Device Objective V2 Object Reference=@n14) -> (Custom Objective Text:Text)
   > SetVisibility(
    self=Get ObjectiveText() -> (ObjectiveText:Text Object Reference)
    InVisibility=Select(Option 0:ESlateVisibility Enum=SelfHitTestInvisible, Option 1:ESlateVisibility Enum=Collapsed, Index=TextIsEmpty(InText=@n15)) -> (ReturnValue:ESlateVisibility Enum)
   )
   > "SetText (Text)"(self=Get ObjectiveText(), InText=@n15)
   Update View Model(execute=@n10.then_2)
   UpdateHealthbarVisibility() {is=title K2Node_CustomEvent}
   > n16 = Get()
   > SetVisibility(
    self=Get HealthBar() -> (HealthBar:Horizontal Box Object Reference)
    InVisibility=Select(Option 0=Hidden, Option 1=HitTestInvisible, Index=CanTakeDamage!(self:Device Objective V2 Object Reference=@n16) -> (True:bool))
   )
   SetVisibility(execute=@n16.else, self=Get HealthBar(), InVisibility=Hidden)
   On Objective Health Reset() {is=title K2Node_CustomEvent}
   > n17 = Get()
   n18 = Truncate(A=GetMaxHealth!(self:Building Actor Object Reference=@n17) -> (ReturnValue:float))
   > Update Health!(execute, self, Current Health:int=@n18, Max Health:int=@n18) -> (then)
   OnObjectiveHealthChanged!() -> (OutputDelegate, then, DestructionObject:Building Prop Destruction Object Object Reference, Health:float=~) {is=title K2Node_CustomEvent}
   > UpdateHealthbarVisibility() {is=title K2Node_CallFunction}
   > UpdateVisibilityCheck() {is=title K2Node_CallFunction}
   > n19 = Get()
   n20 = Truncate(A=GetMaxHealth(self=@n19))
   > Branch(Condition="Not Equal (Integer)"!(A:int=@n20, B:int=0) -> (ReturnValue:bool) {is=title K2Node_CallFunction})
   > Update Health(Current Health=Truncate(A=GetHealth!(self:Building Actor Object Reference=@n19) -> (ReturnValue:float)), Max Health=@n20)
   Set Custom Widget() {is=title K2Node_CustomEvent}
   > n21 = Get()
   > n22 = Get!(execute) -> (then, else, LoadedCustomWidget:User Widget Object Reference)
   > n23 = RemoveFromParent(self=@n22)
   n24 = Get CustomWidget!(self:Device Objective V2 Object Reference=@n21) -> (CustomWidget:User Widget Soft Class Reference)
   n25 = Branch(
    execute=[@n23.then, @n22.else]
    Condition=AND Boolean(
     A=IsValidSoftClassReference!(SoftClassReference:"Object Soft Class Reference (by ref)"=@n24) -> (ReturnValue:bool)
     B=Get CustomWidget_Override!(self:Device Objective V2 Object Reference=@n21) -> (CustomWidget_Override:bool)
    )
   )
   > n26 = CreateWidgetAsync!(
    execute:Exec
    UserWidgetSoftClass:User Widget Soft Class Reference=@n24
    OwningPlayer:Player Controller Object Reference
    bSuspendInputUntilComplete:bool=false
   ) -> (then:Exec, OnComplete:Exec, OnFailure:Exec, UserWidget:User Widget Object Reference) {is=title K2Node_AsyncAction}
   n27 = Set LoadedCustomWidget(execute=@n26.OnComplete, LoadedCustomWidget:User Widget Object Reference=@n26) -> (Output_Get:User Widget Object Reference)
   > n28 = AddChildToOverlay!(execute, self:Overlay Object Reference=Get Custom() -> (Custom:Overlay Object Reference), Content:Widget Object Reference=@n27) -> (then, ReturnValue:Overlay Slot Object Reference)
   > SetVerticalAlignment(self=@n28, InVerticalAlignment=VAlign_Center)
   > SetHorizontalAlignment(self=@n28, InHorizontalAlignment=HAlign_Center)
   > Set RenderTransformPivot!(execute, RenderTransformPivot:Vector 2D Structure=(X=0.5,Y=0.5), self:Widget Object Reference=@n27) -> (then, Output_Get:Vector 2D Structure)
   > Update View Model()
   n29 = Get WidgetSwitcher_0() -> (WidgetSwitcher_0:Widget Switcher Object Reference)
   > Set ActiveWidgetIndex!(execute, ActiveWidgetIndex:int=0, self:Widget Switcher Object Reference=@n29) -> (then, Output_Get:int=0)
   Set LoadedCustomWidget(execute=@n25.else)
   > Set ViewModel(ViewModel:Device - Objective View Model Object Reference) -> (Output_Get:Device - Objective View Model Object Reference)
   > Set ActiveWidgetIndex(ActiveWidgetIndex=1, self=@n29) -> (Output_Get=0)
   n30 = BindViewTargetEvents!() -> (OutputDelegate, then, ViewTarget:Actor Object Reference) {is=title K2Node_CustomEvent}
   > n31 = Cast To Pawn!(execute, Object=@n30.ViewTarget) -> (then, CastFailed, AsPawn:Pawn Object Reference) {is=title K2Node_DynamicCast}
   > n32 = Cast To FortPlayerStateAthena!(execute, Object=Get PlayerState!(self:Pawn Object Reference=@n31) -> (PlayerState:Player State Object Reference)) -> (then, CastFailed, AsFort Player State Athena:Fort Player State Athena Object Reference) {is=title K2Node_DynamicCast}
   > Bind Event to On Team Index Changed Delegate(self:Fort Player State Athena Object Reference=@n32, Delegate=Create Event()) {is=title K2Node_AddDelegate}
   > CheckTeamUpdates()
   n33 = UnbindObjectiveDelegates!() -> (OutputDelegate, then, Objective:Device Objective V2 Object Reference) {is=title K2Node_CustomEvent}
   > Is Valid(InputObject=@n33.Objective)
   > Unbind Event from On Building Attribute Set Health Changed Delegate(self:Building Prop Destruction Object Object Reference=@n33.Objective, Delegate=Create Event()) {is=title K2Node_RemoveDelegate}
   > Unbind Event from Update Refresh(self:Device Objective V2 Object Reference=@n33.Objective, Delegate=Create Event()) {is=title K2Node_RemoveDelegate}
   > Unbind Event from Local Player Changed Teams(self:Device Objective V2 Object Reference=@n33.Objective, Delegate=Create Event()) {is=title K2Node_RemoveDelegate}
   > n34 = Get!(execute, self:Device Objective V2 Object Reference=@n33.Objective) -> (then, else, FortMinigameObjective:Fort Minigame Objective Component Object Reference)
   > Unbind Event from On Objective Updated(self:Fort Minigame Objective Component Object Reference=@n34, Delegate=Create Event()) {is=title K2Node_RemoveDelegate}
   n35 = BindObjectiveDelegates!() -> (OutputDelegate, then, Objective:Device Objective V2 Object Reference) {is=title K2Node_CustomEvent}
   > Is Valid(InputObject=@n35.Objective)
   > Bind Event to On Building Attribute Set Health Changed Delegate(self:Building Prop Destruction Object Object Reference=@n35.Objective, Delegate=Create Event()) {is=title K2Node_AddDelegate}
   > OnObjectiveHealthChanged!(execute, self, DestructionObject:Building Prop Destruction Object Object Reference, Health:float=0) -> (then) {is=title K2Node_CallFunction}
   > Bind Event to Update Refresh(self:Device Objective V2 Object Reference=@n35.Objective, Delegate=Create Event()) {is=title K2Node_AddDelegate}
   > Bind Event to Local Player Changed Teams(self:Device Objective V2 Object Reference=@n35.Objective, Delegate=Create Event()) {is=title K2Node_AddDelegate}
   > CheckTeamUpdates()
   > n36 = Get(self=@n35.Objective)
   > Bind Event to On Objective Updated(self:Fort Minigame Objective Component Object Reference=@n36, Delegate=Create Event()) {is=title K2Node_AddDelegate}
   > OnObjectiveUpdated!(execute, self, Objective:Fort Minigame Objective Component Object Reference=@n36) -> (then)
   OnObjectiveUpdated!() -> (OutputDelegate, then, Objective:Fort Minigame Objective Component Object Reference) {is=title K2Node_CustomEvent}
   > Assign Indicator Icon() {is=title K2Node_CallFunction}
   n37 = UpdateVisibilityTimer!() -> (OutputDelegate, then, Visible:bool=~) {is=title K2Node_CustomEvent}
   > n38 = Branch(Condition=@n37.Visible)
   n39 = Get VisCheckTimerHandle() -> (VisCheckTimerHandle:Timer Handle Structure)
   > Unpause Timer by Handle!(execute, Handle:Timer Handle Structure=@n39) -> (then)
   Pause Timer by Handle!(execute=@n38.else, Handle:Timer Handle Structure=@n39) -> (then)
   n40 = TryUpdate!() -> (OutputDelegate, then, Force Max Health:bool=~) {is=title K2Node_CustomEvent}
   > UpdateVisibilityCheck()
   > n41 = Branch(Condition=@n40.Force Max Health)
   > n42 = On Objective Health Reset() {is=title K2Node_CallFunction}
   Assign Indicator Icon(execute=[@n42.then, @n43.then])
   n43 = OnObjectiveHealthChanged(execute=@n41.else, Health=0)
   CheckTeamUpdates() {is=title K2Node_CustomEvent}
   > Get!(execute) -> (then, else, CurrentViewTarget:Actor Object Reference)
   > Get!(execute) -> (then, else, CurrentObjective:Device Objective V2 Object Reference)
   n44 = GetTeamAffiliation!(ActorA:Actor Object Reference=Get CurrentViewTarget(), ActorB:Actor Object Reference=Get CurrentObjective()) -> (ReturnValue:EFortTeamAffiliation Enum=Friendly)
   > Update Colors!(execute, self, InputPin:EFortTeamAffiliation Enum=@n44) -> (then)
   > Set bFriendly(bFriendly:bool="Equal (Enum)"(A:EFortTeamAffiliation Enum=@n44, B:EFortTeamAffiliation Enum=Friendly)) -> (Output_Get:bool)
   > UpdateVisibilityCheck()
   n45 = OnObjectiveChanged!() -> (OutputDelegate, then, OldObjective:Device Objective V2 Object Reference) {is=title K2Node_CustomEvent}
   > n46 = Sequence()
   > UnbindObjectiveDelegates!(execute, self, Objective:Device Objective V2 Object Reference=@n45.OldObjective) -> (then) {is=title K2Node_CallFunction}
   BindObjectiveDelegates!(execute=@n46.then_1, self, Objective:Device Objective V2 Object Reference=Get CurrentObjective()) -> (then) {is=title K2Node_CallFunction}
   UpdateVisibilityCheck(execute=@n46.then_2)
   StaggerVisibility() {is=title K2Node_CustomEvent}
   > Delay(Duration=RandomFloatInRange(Min=0, Max=1))
   > n47 = Set Timer by Event(Delegate=Create Event(), Time=0.1, bLooping=true)
   > Set VisCheckTimerHandle(VisCheckTimerHandle:Timer Handle Structure=@n47) -> (Output_Get:Timer Handle Structure)
   > SetShouldTickHealthBarComponent(self:Device Objective V2 Object Reference=Get CurrentObjective())
   n48 = UnbindViewTargetEvents!() -> (OutputDelegate, then, ViewTarget:Actor Object Reference) {is=title K2Node_CustomEvent}
   > n49 = Cast To Pawn(Object=@n48.ViewTarget)
   > n50 = Cast To FortPlayerStateAthena(Object=Get PlayerState(self=@n49))
   > Unbind Event from On Team Index Changed Delegate(self:Fort Player State Athena Object Reference=@n50, Delegate=Create Event()) {is=title K2Node_RemoveDelegate}
   n51 = Event InitFromObject!() -> (OutputDelegate, then, InitObject:Object Reference) {is=title K2Node_Event}
   > n52 = Sequence!(execute) -> (then_0, then_1)
   > Set PreviousObjective(PreviousObjective:Device Objective V2 Object Reference=Get CurrentObjective()) -> (Output_Get:Device Objective V2 Object Reference)
   > n53 = Set CurrentObjective(CurrentObjective:Device Objective V2 Object Reference=Cast To Device_Objective_V2!(Object=@n51.InitObject) -> (AsDevice Objective V2:Device Objective V2 Object Reference, bSuccess:bool=~) {is=title K2Node_DynamicCast}) -> (Output_Get:Device Objective V2 Object Reference)
   > Branch(Condition="Not Equal (Object)"!(A:Object Reference=@n53, B:Object Reference=Get PreviousObjective() -> (PreviousObjective:Device Objective V2 Object Reference)) -> (ReturnValue:bool))
   > OnObjectiveChanged!(execute, self, OldObjective:Device Objective V2 Object Reference=Get PreviousObjective()) -> (then) {is=title K2Node_CallFunction}
   StaggerVisibility(execute=@n52.then_1) {is=title K2Node_CallFunction}
   > Set Custom Widget() {is=title K2Node_CallFunction}
   Event Construct()
   > n54 = Sequence()
   > Bind Event to On Local Player View Target Changed(
    self:Fort HUDContext Object Reference=Get FortHUDContext!() -> (ReturnValue:Fort HUDContext Object Reference) {is=title K2Node_GetBlueprintContextEx}
    Delegate=Create Event()
   ) {is=title K2Node_AddDelegate}
   > OnLocalPlayerViewTargetChanged() {is=title K2Node_CallFunction}
   SetVisibility(execute=@n54.then_1, InVisibility=Get PendingVisibility() -> (PendingVisibility=Visible))
   SetVisibility(execute=@n8.then, InVisibility=Get PendingVisibility() -> (PendingVisibility=Visible))
  ]}
 GetDesiredVisibility {nodes [
   GetDesiredVisibility() {is=title K2Node_FunctionEntry}
   > Set NewDesiredVisibility(NewDesiredVisibility:ESlateVisibility Enum=Get HideVisibility() -> (HideVisibility=Visible)) -> (Output_Get:ESlateVisibility Enum=Visible)
   > n1 = Sequence()
   > n2 = Get()
   > n3 = Switch on E_DestructionObjectUIHideCondition!(
    execute
    Selection:E_DestructionObjectUIHideCondition Enum=GetUI HideCondition!(self:Device Objective V2 Object Reference=@n2) -> (HideCondition:E_DestructionObjectUIHideCondition Enum=NewEnumerator0)
   ) -> (NewEnumerator0:Exec, NewEnumerator1:Exec, NewEnumerator2:Exec) {is=title K2Node_SwitchEnum}
   > Set NewDesiredVisibility(NewDesiredVisibility=Get ShowVisibility() -> (ShowVisibility=Visible)) -> (Output_Get=Visible)
   n4 = GetViewTargetDistanceFromActor!(Actor:Actor Object Reference=Get CurrentObjective()) -> (ReturnValue:double)
   n5 = GetHideDistance!() -> (ReturnValue:double)
   Set NewDesiredVisibility(
    execute=@n3.NewEnumerator2
    NewDesiredVisibility=Select(
     Option 0=Get ShowVisibility() -> (ShowVisibility=Visible)
     Option 1=Get HideVisibility() -> (HideVisibility=Visible)
     Index="float >= float"(A=@n4, B=@n5) {is=title K2Node_CallFunction}
    )
   ) -> (Output_Get=Visible)
   Set NewDesiredVisibility(execute=@n3.NewEnumerator1, NewDesiredVisibility=Get HideVisibility() -> (HideVisibility=Visible)) -> (Output_Get=Visible)
   Return Node!(execute=@n1.then_1, ReturnValue:ESlateVisibility Enum=Get NewDesiredVisibility() -> (NewDesiredVisibility:ESlateVisibility Enum=Visible))
  ]}
 GetHideDistance {nodes [
   GetHideDistance() {is=title K2Node_FunctionEntry}
   > n1 = Get()
   > n2 = GetLocalHideDistance!(execute, self:Device Objective V2 Object Reference=@n1) -> (then, Distance:double)
   > Return Node(ReturnValue:double=@n2)
   Return Node(execute=@n1.else, ReturnValue=0)
  ]}
 GetViewTarget {nodes [
   GetViewTarget() {is=title K2Node_FunctionEntry}
   > n1 = Sequence()
   n2 = GetOwningPlayer()
   > Is Valid(InputObject=@n2)
   n3 = GetViewTarget!(self:Controller Object Reference=@n2) -> (ReturnValue:Actor Object Reference) {is=title K2Node_CallFunction}
   > Is Valid(InputObject=@n3)
   > Return Node!(execute, ViewTarget:Actor Object Reference=@n3)
   Return Node(execute=@n1.then_1)
  ]}
 GetViewTargetDistanceFromActor {nodes [
   GetViewTargetDistanceFromActor!() -> (then, Actor:Actor Object Reference) {is=title K2Node_FunctionEntry}
   > n1 = Sequence()
   > Get!(execute) -> (then, else, Actor:Actor Object Reference)
   > n2 = Get!(execute) -> (then, else, CurrentViewTarget:Actor Object Reference)
   n3 = Get Actor Location(self=Get Actor() -> (Actor:Actor Object Reference))
   n4 = Get Actor Location(self=@n2)
   > Return Node!(
    execute
    ReturnValue:double=Vector Length(A=vector - vector!(A:Vector=@n3, B:Vector=@n4) -> (ReturnValue:Vector) {is=title K2Node_CallFunction})
   )
   Return Node(execute=@n1.then_1, ReturnValue=0)
  ]}
 IsObjectiveVisible {nodes [
   IsObjectiveVisible() {is=title K2Node_FunctionEntry}
   > n1 = Sequence()
   > n2 = Get!(execute) -> (then, else, CurrentObjective:Device Objective V2 Object Reference)
   > n3 = Get!(execute, self:Building Static Mesh Object Reference=@n2) -> (then, else, StaticMeshComponent:Base Building Static Mesh Component Object Reference)
   > Return Node(ReturnValue:bool=IsVisible(self:Scene Component Object Reference=@n3))
   Return Node(execute=@n1.then_1, ReturnValue=false)
  ]}
 IsShowingHealthbarAllowedBasedOnMinigameState {nodes [
   IsShowingHealthbarAllowedBasedOnMinigameState() {is=title K2Node_FunctionEntry}
   > n1 = Sequence()
   > n2 = Get!(execute) -> (then, else, CurrentObjective:Device Objective V2 Object Reference)
   > n3 = Get!(execute, self:Device Objective V2 Object Reference=@n2) -> (then, else, FortMinigameObjective:Fort Minigame Objective Component Object Reference)
   > n4 = Get!(execute, self:Fort Minigame Logic Component Object Reference=@n3) -> (then, else, CurrentMinigame:Fort Minigame Object Reference)
   > Set bOutAllowShowHealthbar(bOutAllowShowHealthbar:bool=Select!(
     PreGame:bool=true
     Setup:bool=false
     Transitioning:bool=false
     WaitingForCameras:bool=false
     Warmup:bool=false
     InProgress:bool=true
     PostGameTimeDilation:bool=false
     PostRoundEnd:bool=false
     PostGameEnd:bool=false
     PostGameAbandon:bool=false
     PostGameReset:bool=false
     Teardown:bool=false
     Index:EFortMinigameState Enum=Get CurrentState!(self:Fort Minigame Object Reference=@n4) -> (CurrentState:EFortMinigameState Enum=PreGame)
    ) -> (ReturnValue:bool)) -> (Output_Get:bool)
   Return Node!(execute=@n1.then_1, bAllowShowingHealthbar:bool=Get bOutAllowShowHealthbar() -> (bOutAllowShowHealthbar:bool))
  ]}
 ShouldHideWithoutAnimation {nodes [
   ShouldHideWithoutAnimation() {is=title K2Node_FunctionEntry}
   > n1 = Sequence()
   > Branch(Condition=CanDetermineVisibility!() -> (bIsDeterminable:bool) {is=title K2Node_CallFunction})
   > Branch(Condition=IsShowingHealthbarAllowedBasedOnMinigameState!() -> (bAllowShowingHealthbar:bool) {is=title K2Node_CallFunction})
   > Branch(Condition=IsObjectiveVisible!() -> (ReturnValue:bool) {is=title K2Node_CallFunction})
   > Return Node!(execute, bReturnValue:bool=false)
   Return Node(execute=@n1.then_1, bReturnValue=true)
  ]}
 Update Colors {nodes [
   n1 = Update Colors!() -> (then, InputPin:EFortTeamAffiliation Enum=Friendly) {is=title K2Node_FunctionEntry}
   > n2 = GetDynamicMaterial(self=Get ProgressIcon() -> (ProgressIcon:Image Object Reference))
   n3 = Select!(
    Friendly:Linear Color Structure=(R=0.004,G=0.0232,B=0.2159,A=1)
    Neutral:Linear Color Structure=(R=0.0243,G=0.0243,B=0.0243,A=1)
    Hostile:Linear Color Structure=(R=0.1845,G=0,B=0,A=1)
    Index:EFortTeamAffiliation Enum=@n1
   ) -> (ReturnValue:Linear Color Structure)
   > SetVectorParameterValue(self=@n2, ParameterName=BackingColor, Value=@n3)
   n4 = Select(Friendly=(R=0.0513,G=0.7379,B=1,A=1), Neutral=(R=1,G=1,B=1,A=1), Hostile=(R=1,G=0.4,B=0.1,A=1), Index=@n1)
   > SetVectorParameterValue(self=@n2, ParameterName=BorderColor, Value=@n4)
   n5 = Select(Friendly=(R=0,G=0.402,B=1,A=1), Neutral=(R=0.176,G=0.729,B=0.0784,A=1), Hostile=(R=0.875,G=0.1072,B=0.0219,A=1), Index=@n1)
   > SetVectorParameterValue(self=@n2, ParameterName=FillColor, Value=@n5)
   n6 = Select(Friendly=(R=0.5029,G=1,B=1,A=1), Neutral=(R=0.7875,G=1,B=0.75,A=1), Hostile=(R=1,G=0.7913,B=0.5029,A=1), Index=@n1)
   > SetContentColorAndOpacity!(
    execute
    self:Border Object Reference=Get IndicatorColor() -> (IndicatorColor:Border Object Reference)
    InContentColorAndOpacity:Linear Color Structure=@n6
   ) -> (then)
   > n7 = Get!(execute) -> (then, else, ViewModel:Device - Objective View Model Object Reference)
   > Update Colors!(
    execute
    self:Device - Objective View Model Object Reference=@n7
    Color_Backing:Linear Color Structure=@n3
    Color_Border:Linear Color Structure=@n4
    Color_Fill:Linear Color Structure=@n5
    Color_Indicator:Linear Color Structure=@n6
   ) -> (then) {is=title K2Node_CallFunction}
  ]}
 Update Health {nodes [
   n1 = Update Health!() -> (then, Current Health:int, Max Health:int) {is=title K2Node_FunctionEntry}
   > n2 = Sequence()
   > Set HealthCurrent(HealthCurrent:int=@n1) -> (Output_Get:int=0)
   > Set HealthMax(HealthMax:int=@n1.Max Health) -> (Output_Get:int=0)
   n3 = Get CurrentObjective()
   n4 = "To Float (Integer)"(InInt=@n1.Max Health)
   n5 = "To Float (Integer)"(InInt=@n1)
   > Set local HP Percent(local HP Percent:double=SelectFloat(A=float / float(A=@n5, B=@n4) {is=title K2Node_CallFunction}, B=1, bPickA=CanTakeDamage(self=@n3))) -> (Output_Get:double)
   > "SetText (Text)"(self=Get CurHP() -> (CurHP:Text Object Reference), InText="To Text (Integer)"(Value=Get HealthCurrent() -> (HealthCurrent:int=0)))
   > "SetText (Text)"(self=Get MaxHP() -> (MaxHP:Text Object Reference), InText="To Text (Integer)"(Value=Get HealthMax() -> (HealthMax:int=0)))
   > n6 = GetDynamicMaterial(self=Get ProgressIcon())
   > SetScalarParameterValue(self=@n6, ParameterName=PercentFill, Value=Get local HP Percent() -> (local HP Percent:double))
   > n7 = Branch(Condition="float < float"!(A:double=Get local HP Percent(), B:double=Get CurPercent() -> (CurPercent:double)) -> (ReturnValue:bool) {is=title K2Node_CallFunction})
   > Set CurPercent(CurPercent:double=Get local HP Percent()) -> (Output_Get:double)
   n8 = Select!(
    Option 0:Widget Animation Object Reference=Get EnemyObjectiveDamaged() -> (EnemyObjectiveDamaged:Widget Animation Object Reference)
    Option 1:Widget Animation Object Reference=Get FriendlyObjectiveDamaged() -> (FriendlyObjectiveDamaged:Widget Animation Object Reference)
    Index:bool=Get bFriendly() -> (bFriendly:bool)
   ) -> (ReturnValue:Widget Animation Object Reference)
   > Branch(Condition=NOT Boolean(A=IsAnimationPlaying(InAnimation=@n8)))
   > PlayAnimation(InAnimation=@n8)
   Set CurPercent(execute=@n7.else, CurPercent=Get local HP Percent())
   Update View Model(execute=@n2.then_1)
  ]}
 Update Offscreen Arrow {nodes [
   Update Offscreen Arrow() {is=title K2Node_FunctionEntry}
   > n1 = Get!(execute) -> (then, else, CurrentObjective:Device Objective V2 Object Reference)
   > Branch(Condition=Get ClampToScreen!(self:Device Objective V2 Object Reference=@n1) -> (ClampToScreen:bool))
   > Set WidgetComp(WidgetComp:Fort Widget Component Object Reference=Get HealthBarComp!(self:Device Objective V2 Object Reference=@n1) -> (HealthBarComp:Fort Widget Component Object Reference)) -> (Output_Get:Fort Widget Component Object Reference)
   > Set ViewportSize(ViewportSize:Vector 2D Structure=GetViewportSize!() -> (ReturnValue:Vector 2D Structure)) -> (Output_Get:Vector 2D Structure)
   n2 = Get WidgetComp() -> (WidgetComp:Fort Widget Component Object Reference)
   n3 = Get!(self:Fort Widget Component Object Reference=@n2) -> (PreClampedLocalPosition_X:double, PreClampedLocalPosition_Y:double)
   n4 = GetViewportScale!() -> (ReturnValue:float)
   n5 = Get!() -> (ViewportSize_X:double, ViewportSize_Y:double)
   n6 = float / float(A=@n5, B=@n4) {is=title K2Node_PromotableOperator}
   n7 = Get!(self:Fort Widget Component Object Reference=@n2) -> (PreClampedLocalPosition_X:double, PreClampedLocalPosition_Y:double)
   n8 = float / float(A=@n5.ViewportSize_Y, B=@n4)
   > Set bOutOfViewportBounds(bOutOfViewportBounds:bool=OR Boolean!(
     A="float < float"(A=@n3, B=0)
     B="float < float"(A=@n3.PreClampedLocalPosition_Y, B=0)
     C:bool="float > float"(A=@n7, B=@n6) {is=title K2Node_CallFunction}
     D:bool="float > float"(A=@n7.PreClampedLocalPosition_Y, B=@n8)
    ) -> (ReturnValue)) -> (Output_Get:bool)
   n9 = AND Boolean(
    A=Get bOutOfViewportBounds() -> (bOutOfViewportBounds:bool)
    B=Get ShowOffscreenArrow!(self:Device Objective V2 Object Reference=@n1) -> (ShowOffscreenArrow:bool)
   )
   > SetVisibility(
    self=Get ArrowSizeBox() -> (ArrowSizeBox:Size Box Object Reference)
    InVisibility=Select(Option 0:ESlateVisibility Enum=Collapsed, Option 1:ESlateVisibility Enum=SelfHitTestInvisible, Index=@n9) -> (ReturnValue:ESlateVisibility Enum)
   )
   > Branch(Condition=@n9)
   > Branch(Condition=Get bOutOfViewportBounds())
   n10 = Get WidgetComp()
   n11 = Get(self=@n10)
   > n12 = Branch(Condition="float < float"(A=@n11, B=0))
   n13 = Get ArrowSizeBox()
   > SetRenderTransformAngle!(execute, self:Widget Object Reference=@n13, Angle:float=-90) -> (then)
   n14 = Branch(execute=@n12.else, Condition="float < float"(A=@n11.PreClampedLocalPosition_Y, B=0))
   > SetRenderTransformAngle(self=@n13, Angle=0)
   n15 = Get(self=@n10)
   n16 = Get!() -> (ViewportSize_X:double, ViewportSize_Y:double)
   n17 = Branch(execute=@n14.else, Condition="float > float"(A=@n15, B=@n16))
   > SetRenderTransformAngle(self=@n13, Angle=90)
   Branch(execute=@n17.else, Condition="float > float"(A=@n15.PreClampedLocalPosition_Y, B=@n16.ViewportSize_Y))
   > SetRenderTransformAngle(self=@n13, Angle=180)
  ]}
 Update View Model {nodes [
   Update View Model() {is=title K2Node_FunctionEntry}
   > Get!(execute:Exec) -> (then:Exec, else:Exec, LoadedCustomWidget:User Widget Object Reference)
   > n1 = Get!(execute) -> (then, else, ViewModel:Device - Objective View Model Object Reference)
   SetViewModel!(
    execute=[@n1.then, @n4.then]
    self:MVVM View Object Reference=Get View From User Widget!(
     self:Viewmodel Engine Subsytem Object Reference=Get MVVMSubsystem() -> (ReturnValue:Viewmodel Engine Subsytem Object Reference) {is=title K2Node_GetEngineSubsystem}
     UserWidget:User Widget Object Reference=Get LoadedCustomWidget() -> (LoadedCustomWidget:User Widget Object Reference)
    ) -> (ReturnValue:MVVM View Object Reference) {is=title K2Node_CallFunction}
    ViewModelName:Name=UEFN_ObjectiveBadge_ViewModel
    ViewModel:Notify Field Value Changed Interface=Get ViewModel() -> (ViewModel:Device - Objective View Model Object Reference)
   ) -> (then, ReturnValue:bool)
   > n2 = Get!(execute) -> (then, else, CurrentObjective:Device Objective V2 Object Reference)
   > Update Details!(
    execute
    self:Device - Objective View Model Object Reference=Get ViewModel()
    Text_Label:Text=Get Custom Objective Text(self=@n2)
    Icon_Material:Texture 2D Object Reference=Get LoadedIconTexture() -> (LoadedIconTexture:Texture 2D Object Reference)
    CurrentHealth:Integer64="To Integer64 (Integer)"!(InInt:int=Get HealthCurrent() -> (HealthCurrent=0)) -> (ReturnValue:Integer64=0)
    MaxHealth:Integer64="To Integer64 (Integer)"(InInt=Get HealthMax() -> (HealthMax=0)) -> (ReturnValue=0)
    PercentHealth:double=Get CurPercent()
   ) -> (then)
   > CheckTeamUpdates() {is=title K2Node_CallFunction}
   n3 = Construct Object from Class(execute=@n1.else, Class=&/UEFN_ObjectiveBadge_ViewModel.UEFN_ObjectiveBadge_ViewModel_C) -> (ReturnValue:Device - Objective View Model Object Reference)
   > n4 = Set ViewModel(ViewModel=@n3)
  ]}
]
variables [
 bFriendly {type=boolean}
 CurPercent {type=number}
 CurrentObjective {type=Device_Objective_V2_C, default=null}
 CurrentViewTarget {type=Actor, default=null}
 DesiredVisibility {type=ESlateVisibility, default=Collapsed}
 HealthCurrent {type=integer}
 HealthMax {type=integer}
 HideTimer {type=TimerHandle}
 HideVisibility {type=ESlateVisibility, default=Collapsed}
 IndicatorLabelMap {default {0=A, 1=B, 2=C, 3=D, 4=W, 5=X, 6=Y, 7=Z, 8=""}}
 LoadedCustomWidget {type=UserWidget, default=null}
 LoadedIconTexture {type=Texture2D}
 PendingVisibility {type=ESlateVisibility, default=Collapsed}
 PreviousObjective {type=Device_Objective_V2_C, default=null}
 ShowVisibility {type=ESlateVisibility, default=SelfHitTestInvisible}
 ViewModel {type=UEFN_ObjectiveBadge_ViewModel_C, default=null}
 VisCheckTimerHandle {type=TimerHandle}
]
widgets {
 info {inheritedWidgetCount=0, parentClass=&UserWidget, rootWidgetClass=&SizeBox, widgetCount=18}
 widgets [
  SizeBox_3 {widgetClassPath=&SizeBox, widget=SizeBox_3, properties {heightOverride=21.3612}}
  Overlay_2 {widgetClassPath=&Overlay, parent=SizeBox_3, widget=Overlay_2, slot {class=SizeBoxSlot}}
  WidgetSwitcher_0 {
   widgetClassPath=&WidgetSwitcher
   bIsVariable=true
   parent=Overlay_2
   widget=WidgetSwitcher_0
   slot {class=OverlaySlot, properties {horizontalAlignment=HAlign_Center, verticalAlignment=VAlign_Center}}
  }
  Custom {widgetClassPath=&Overlay, bIsVariable=true, parent=WidgetSwitcher_0, widget=Custom, slot {class=WidgetSwitcherSlot}}
  Default {widgetClassPath=&VerticalBox, parent=WidgetSwitcher_0, widget=Default, slot {class=WidgetSwitcherSlot}}
  HorizontalBox_58 {widgetClassPath=&HorizontalBox, parent=Default, widget=HorizontalBox_58, slot {class=VerticalBoxSlot}}
  ObjectiveText {
   widgetClassPath=&TextBlock
   bIsVariable=true
   parent=HorizontalBox_58
   widget=ObjectiveText
   properties {
    justification=Center
    minDesiredWidth=20
    visibility=Collapsed
    font {
     bForceMonospaced=false
     bMaterialIsStencil=false
     fontMaterial=null
     fontObject=&Burbank
     letterSpacing=0
     monospacedWidth=1
     size=20
     skewAmount=0
     typefaceFontName=Burbank-PlayerPower
     outlineSettings {
      bApplyOutlineToDropShadows=false
      bMiteredCorners=false
      bSeparateFillAlpha=false
      outlineMaterial=null
      outlineSize=1
      outlineColor=<rgba 0 0 0 1>
     }
    }
   }
   slot {class=HorizontalBoxSlot, properties {horizontalAlignment=HAlign_Center, verticalAlignment=VAlign_Center, size {sizeRule=Fill, value=1}}}
  }
  HealthBar {
   widgetClassPath=&HorizontalBox
   bIsVariable=true
   parent=Default
   widget=HealthBar
   slot {class=VerticalBoxSlot, properties {horizontalAlignment=HAlign_Center, padding=<ltrb 0 0 0 -2>}}
  }
  CurHP {
   widgetClassPath=&TextBlock
   bIsVariable=true
   parent=HealthBar
   widget=CurHP
   properties {
    justification=Right
    minDesiredWidth=20
    renderOpacity=0.75
    text=75
    font {
     bForceMonospaced=false
     bMaterialIsStencil=false
     fontMaterial=null
     fontObject=&Roboto
     letterSpacing=0
     monospacedWidth=1
     size=10
     skewAmount=0
     typefaceFontName=Bold
     outlineSettings {
      bApplyOutlineToDropShadows=false
      bMiteredCorners=false
      bSeparateFillAlpha=false
      outlineMaterial=null
      outlineSize=0
      outlineColor=<rgba 0 0 0 1>
     }
    }
   }
   slot {class=HorizontalBoxSlot, properties {padding=<ltrb 0 0 2 0>}}
  }
  Image_1 {
   widgetClassPath=&Image
   bIsVariable=true
   parent=HealthBar
   widget=Image_1
   properties {
    renderOpacity=0.5
    visibility=SelfHitTestInvisible
    brush {
     bIsDynamicallyLoaded=false
     drawAs=Image
     imageType=NoImage
     mirroring=NoMirror
     resourceName=None
     resourceObject=null
     tiling=NoTile
     imageSize=<1 6>
     margin=<ltrb 0 0 0 0>
     outlineSettings {
      bUseBrushTransparency=false
      roundingType=HalfHeightRadius
      width=0
      color {colorUseRule=UseColor_Specified, specifiedColor=<rgba 0 0 0 0>}
      cornerRadii=<0 0 0 0>
     }
     tintColor {colorUseRule=UseColor_Specified, specifiedColor=<rgba 1 1 1 1>}
     uVRegion {bIsValid=false, max=<0 0>, min=<0 0>}
    }
   }
   slot {class=HorizontalBoxSlot, properties {padding=<ltrb 0 1 0 2>}}
  }
  MaxHP {
   widgetClassPath=&TextBlock
   bIsVariable=true
   parent=HealthBar
   widget=MaxHP
   properties {
    minDesiredWidth=20
    renderOpacity=0.75
    text=75
    font {
     bForceMonospaced=false
     bMaterialIsStencil=false
     fontMaterial=null
     fontObject=&Roboto
     letterSpacing=0
     monospacedWidth=1
     size=10
     skewAmount=0
     typefaceFontName=Bold
     outlineSettings {
      bApplyOutlineToDropShadows=false
      bMiteredCorners=false
      bSeparateFillAlpha=false
      outlineMaterial=null
      outlineSize=0
      outlineColor=<rgba 0 0 0 1>
     }
    }
   }
   slot {class=HorizontalBoxSlot, properties {padding=<ltrb 2 0 0 0>}}
  }
  DefaultWidgetContainer {widgetClassPath=&Overlay, parent=Default, widget=DefaultWidgetContainer, slot {class=VerticalBoxSlot}}
  ProgressIcon {
   widgetClassPath=&Image
   bIsVariable=true
   parent=DefaultWidgetContainer
   widget=ProgressIcon
   properties {
    brush {
     bIsDynamicallyLoaded=false
     drawAs=Image
     imageType=NoImage
     mirroring=NoMirror
     resourceName=None
     resourceObject=/Game/Athena/HUD/LTM/Playground/Materials/M_UI_HUD_Objective_Progress
     tiling=NoTile
     imageSize=<58 58>
     margin=<ltrb 0 0 0 0>
     outlineSettings {
      bUseBrushTransparency=false
      roundingType=HalfHeightRadius
      width=0
      color {colorUseRule=UseColor_Specified, specifiedColor=<rgba 0 0 0 0>}
      cornerRadii=<0 0 0 0>
     }
     tintColor {colorUseRule=UseColor_Specified, specifiedColor=<rgba 1 1 1 1>}
     uVRegion {bIsValid=false, max=<0 0>, min=<0 0>}
    }
    renderTransformPivot=<0.5 1>
   }
   slot {class=OverlaySlot, properties {horizontalAlignment=HAlign_Center, verticalAlignment=VAlign_Center}}
  }
  IndicatorColor {
   widgetClassPath=&Border
   bIsVariable=true
   parent=DefaultWidgetContainer
   widget=IndicatorColor
   properties {
    horizontalAlignment=HAlign_Center
    verticalAlignment=VAlign_Center
    background {
     bIsDynamicallyLoaded=false
     drawAs=NoDrawType
     imageType=NoImage
     mirroring=NoMirror
     resourceName=None
     resourceObject=null
     tiling=NoTile
     imageSize=<32 32>
     margin=<ltrb 0 0 0 0>
     outlineSettings {
      bUseBrushTransparency=false
      roundingType=HalfHeightRadius
      width=0
      color {colorUseRule=UseColor_Specified, specifiedColor=<rgba 0 0 0 0>}
      cornerRadii=<0 0 0 0>
     }
     tintColor {colorUseRule=UseColor_Specified, specifiedColor=<rgba 1 1 1 1>}
     uVRegion {bIsValid=false, max=<0 0>, min=<0 0>}
    }
    padding=<ltrb 0 0 0 0>
   }
   slot {class=OverlaySlot, properties {horizontalAlignment=HAlign_Center, verticalAlignment=VAlign_Center, padding=<ltrb 0 0 0 2>}}
  }
  IndicatorIcon {
   widgetClassPath=&Image
   bIsVariable=true
   parent=IndicatorColor
   widget=IndicatorIcon
   properties {brush {
     bIsDynamicallyLoaded=false
     drawAs=Image
     imageType=NoImage
     mirroring=NoMirror
     resourceName=None
     resourceObject=/Game/Athena/UI/InGame/Creative/PropertyEditor/Icon_Picker/Textures/T_UI_IconPicker_EMPTY
     tiling=NoTile
     imageSize=<32 32>
     margin=<ltrb 0 0 0 0>
     outlineSettings {
      bUseBrushTransparency=false
      roundingType=HalfHeightRadius
      width=0
      color {colorUseRule=UseColor_Specified, specifiedColor=<rgba 0 0 0 0>}
      cornerRadii=<0 0 0 0>
     }
     tintColor {colorUseRule=UseColor_Specified, specifiedColor=<rgba 1 1 1 1>}
     uVRegion {bIsValid=false, max=<0 0>, min=<0 0>}
    }}
   slot {class=BorderSlot, properties {horizontalAlignment=HAlign_Center, verticalAlignment=VAlign_Center, padding=<ltrb 0 0 0 0>}}
  }
  ArrowSizeBox {
   widgetClassPath=&SizeBox
   bIsVariable=true
   parent=Overlay_2
   widget=ArrowSizeBox
   properties {bOverride_HeightOverride=true, bOverride_WidthOverride=true, heightOverride=210, visibility=Hidden, widthOverride=210}
   slot {class=OverlaySlot}
  }
  ArrowSpinnerSizeBox {
   widgetClassPath=&SizeBox
   bIsVariable=true
   parent=ArrowSizeBox
   widget=ArrowSpinnerSizeBox
   properties {bOverride_WidthOverride=true, widthOverride=32}
   slot {class=SizeBoxSlot, properties {horizontalAlignment=HAlign_Center}}
  }
  Image_61 {
   widgetClassPath=&Image
   parent=ArrowSpinnerSizeBox
   widget=Image_61
   properties {brush {
     bIsDynamicallyLoaded=false
     drawAs=Image
     imageType=NoImage
     mirroring=NoMirror
     resourceName=None
     resourceObject=/Game/Widgets/SActorCanvas/DirectionArrow
     tiling=NoTile
     imageSize=<32 32>
     margin=<ltrb 0 0 0 0>
     outlineSettings {
      bUseBrushTransparency=false
      roundingType=HalfHeightRadius
      width=0
      color {colorUseRule=UseColor_Specified, specifiedColor=<rgba 0 0 0 0>}
      cornerRadii=<0 0 0 0>
     }
     tintColor {colorUseRule=UseColor_Specified, specifiedColor=<rgba 1 1 1 1>}
     uVRegion {bIsValid=false, max=<0 0>, min=<0 0>}
    }}
   slot {class=SizeBoxSlot, properties {horizontalAlignment=HAlign_Center, verticalAlignment=VAlign_Top}}
  }
 ]
}
