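; View model for the Objective Device badge UI, holding the objective's display state and color scheme.
; Update Details stores the objective name, icon texture, current and max health, and a health percentage passed in by the caller rather than computed here.
; Update Colors stores the backing, border, fill and indicator colors.
; Both custom events are plain pass-through setters with no other logic.
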(event Custom|Update%20Details
  (params (Text_Label :type Text) (Icon_Material :type Texture%202D%20Object%20Reference) (CurrentHealth :type Integer64) (MaxHealth :type Integer64) (PercentHealth :type Float%20%28double-precision%29))
  (Variables|Default|SetObjective%20Name Text_Label)
  (Variables|Default|SetIcon_Texture Icon_Material)
  (Variables|Default|SetCurrentHealth CurrentHealth)
  (Variables|Default|SetMaxHealth MaxHealth)
  (Variables|Default|SetPercentHealth PercentHealth))

(event Custom|Update%20Colors
  (params (Color_Backing :type Linear%20Color%20Structure) (Color_Border :type Linear%20Color%20Structure) (Color_Fill :type Linear%20Color%20Structure) (Color_Indicator :type Linear%20Color%20Structure))
  (Variables|Default|SetColor_Backing Color_Backing)
  (Variables|Default|SetColor_Border Color_Border)
  (Variables|Default|SetColor_Fill Color_Fill)
  (Variables|Default|SetColor_Indicator Color_Indicator))
